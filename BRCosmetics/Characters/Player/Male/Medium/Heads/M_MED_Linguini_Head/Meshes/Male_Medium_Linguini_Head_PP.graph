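# Post-process Anim Blueprint for the Linguini male medium head cosmetic on the Fortnite base head skeleton, shrinking the head slightly while targeting.
# The incoming pose passes through a single Modify Bone node that multiplies the head bone's scale by 0.95 through additive mode.
# The node's bool alpha reads bIsTargeting, which is not declared in this Blueprint, with linear blending.
# The Event Graph holds only unconnected default nodes.
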AnimBlueprint /BRCosmetics/Characters/Player/Male/Medium/Heads/M_MED_Linguini_Head/Meshes/Male_Medium_Linguini_Head_PP
target_skeleton=/Game/Characters/Player/Common/Fortnite_Base_Head/Fortnite_Base_Head_Export_Skeleton
graphs [
 AnimGraph {nodes [
   n1 = "Transform (Modify) Bone - Bone: head"!(
    ComponentPose=Local To Component(LocalPose=Input Pose())
    bAlphaBoolEnabled:bool=Get bIsTargeting()
    Translation="0,0,0"
    Rotation="0,0,0"
    Scale="0.95,0.95,0.95"
   ) -> (Pose) {alphaInputType=Bool, scaleMode=BMM_Additive, alphaBoolBlend {blendOption=Linear}, boneToModify {boneName=head}, scale=<0.95 0.95 0.95>}
   Output Pose(Result=Component To Local(ComponentPose=@n1)) {layerGroup=None}
  ]}
 EventGraph {nodes [
   TryGetPawnOwner()
   Event BlueprintUpdateAnimation()
  ]}
]
variables []
